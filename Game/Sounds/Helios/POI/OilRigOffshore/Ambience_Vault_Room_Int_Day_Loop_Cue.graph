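# Looping, non-spatialized indoor daytime ambience cue for the vault room at the offshore oil rig POI, built from reused Hollow Base and Residential sounds.
# A quality switch picks between two mixes, each layering the basement loop with a pitched-up residential interior loop in a quad or stereo variant.
# Two nested, endlessly looping envelopers of different lengths add slow pitch and subtle volume drift.
# Attenuation and reverb send are off, and concurrency prevents new instances.
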SoundCue /Game/Sounds/Helios/POI/OilRigOffshore/Ambience_Vault_Room_Int_Day_Loop_Cue
root=Enveloper_2
nodes [
 {
  id=Enveloper_1
  children [QualityLevel_0]
  properties {
   bLoop=true
   bLoopIndefinitely=true
   loopEnd=45
   pitchCurve {editorCurveData {keys [{interpMode=RCIM_Cubic, time=0}, <at 22.5 0.91 RCIM_Cubic>, {interpMode=RCIM_Cubic, time=45}]}}
   volumeCurve {editorCurveData {keys [<at 0 1>]}}
  }
 }
 {
  id=Enveloper_2
  children [Enveloper_1]
  properties {
   bLoop=true
   bLoopIndefinitely=true
   loopEnd=20
   pitchCurve {editorCurveData {keys [{time=0}]}}
   volumeCurve {editorCurveData {keys [
      {arriveTangent=-0.01, interpMode=RCIM_Cubic, time=0, value=1}
      <at 10 0.95 RCIM_Cubic>
      {interpMode=RCIM_Cubic, leaveTangent=0.01, time=20, value=1}
     ]}}
  }
 }
 {id=Mixer_0, children [WavePlayer_1,Modulator_0], properties {inputVolume [3,2.5]}}
 {id=Mixer_1, children [WavePlayer_3,Modulator_1], properties {inputVolume [3,2.5]}}
 {id=Modulator_0, children [WavePlayer_0], properties {pitchMax=2, pitchMin=2}}
 {id=Modulator_1, children [WavePlayer_2], properties {pitchMax=2, pitchMin=2}}
 {id=QualityLevel_0, children [Mixer_0,Mixer_0,Mixer_1,Mixer_1]}
 {id=WavePlayer_0, properties {bLooping=true, soundWaveAssetPtr=/Game/Sounds/Asteria/POI/Residential/Ambience_Residential_Interior_Quad_1}}
 {id=WavePlayer_1, properties {bLooping=true, soundWaveAssetPtr=/Game/Sounds/Artemis/POI/Hollow_Base/Ambience_Hollow_Base_Basement_Day_Loop}}
 {id=WavePlayer_2, properties {bLooping=true, soundWaveAssetPtr=/Game/Sounds/Asteria/POI/Residential/Ambience_Residential_Interior_Stereo_1}}
 {id=WavePlayer_3, properties {bLooping=true, soundWaveAssetPtr=&Ambience_Hollow_Base_Basement_Day_Loop}}
]
properties {
 bEnableBusSends=false
 bEnableSubmixSends=false
 bOverrideConcurrency=true
 maxDistance=2097152
 soundClassObject=&Fort_Ambient_Indoors_Athena
 volumeMultiplier=0.501
 attenuationOverrides {bAttenuate=false, bEnableReverbSend=false, bSpatialize=false}
 concurrencyOverrides {resolutionRule=PreventNew}
}
