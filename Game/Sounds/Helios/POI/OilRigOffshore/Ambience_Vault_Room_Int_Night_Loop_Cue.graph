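# Non-spatialized looping night interior ambience cue for the vault room at the Helios offshore oil rig POI, on the indoor ambient sound class with doubled pitch.
# A quality node picks between two mixes, each layering a reused Starlight church ambience over a residential interior bed, quad or stereo by tier.
# Two nested, endlessly looping envelopers with different periods modulate pitch and gently swell volume.
# A concurrency override allows two instances, stopping the oldest.
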SoundCue /Game/Sounds/Helios/POI/OilRigOffshore/Ambience_Vault_Room_Int_Night_Loop_Cue
root=Enveloper_2
nodes [
 {
  id=Enveloper_1
  children [QualityLevel_0]
  properties {
   bLoop=true
   bLoopIndefinitely=true
   loopEnd=30
   pitchCurve {editorCurveData {keys [{interpMode=RCIM_Cubic, time=0}, <at 15.9468 0.9181 RCIM_Cubic>, {interpMode=RCIM_Cubic, time=30}]}}
   volumeCurve {editorCurveData {keys [<at 0 1>]}}
  }
 }
 {
  id=Enveloper_2
  children [Enveloper_1]
  properties {
   bLoop=true
   bLoopIndefinitely=true
   loopEnd=20
   pitchCurve {editorCurveData {keys [{time=0}]}}
   volumeCurve {editorCurveData {keys [
      {arriveTangent=-0.01, interpMode=RCIM_Cubic, time=0, value=1}
      <at 10 0.95 RCIM_Cubic>
      {interpMode=RCIM_Cubic, leaveTangent=0.01, time=20, value=1}
     ]}}
  }
 }
 {id=Mixer_0, children [WavePlayer_2,WavePlayer_1], properties {inputVolume [1.5,1]}}
 {id=Mixer_1, children [WavePlayer_4,WavePlayer_3], properties {inputVolume [1.5,1]}}
 {id=QualityLevel_0, children [Mixer_0,Mixer_0,Mixer_1,Mixer_1]}
 {id=WavePlayer_1, properties {bLooping=true, soundWaveAssetPtr=/Game/Sounds/Asteria/POI/Residential/Ambience_Residential_Interior_Quad_3}}
 {
  id=WavePlayer_2
  properties {bLooping=true, soundWaveAssetPtr=/Game/Sounds/Fort_Starlight/S1/SoundWaves/AMBI_Starlight_mToD_Church_Int_Day_Loop}
 }
 {id=WavePlayer_3, properties {soundWaveAssetPtr=/Game/Sounds/Asteria/POI/Residential/Ambience_Residential_Interior_Stereo_3}}
 {
  id=WavePlayer_4
  properties {bLooping=true, soundWaveAssetPtr=&AMBI_Starlight_mToD_Church_Int_Day_Loop}
 }
]
properties {
 bOverrideConcurrency=true
 maxDistance=2097152
 pitchMultiplier=2
 soundClassObject=&Fort_Ambient_Indoors_Athena
 volumeMultiplier=0.752
 attenuationOverrides {bAttenuate=false, bEnableReverbSend=false, bSpatialize=false}
 concurrencyOverrides {maxCount=2, resolutionRule=StopOldest, voiceStealReleaseTime=2}
}
