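# Sound cue for a one-shot metal stress creak and groan ambience, named as a spot sound for the big metal bridge at the offshore oil rig POI.
# A quality-level node sends its first two levels to an equal-weight random pick of twelve knock and stretch waves, and the other two to three of those knocks.
# The cue reuses generic biome waves, pitched down and quietened, with tile-ambient one-shot attenuation.
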SoundCue /Game/Sounds/Helios/POI/OilRigOffshore/Spot_OilRig_Big_Metal_Bridge_Stress_Creak_Cue
root=QualityLevel_0
nodes [
 {id=QualityLevel_0, children [Random_0,Random_0,Random_1,Random_1]}
 {
  id=Random_0
  children [
   WavePlayer_1
   WavePlayer_2
   WavePlayer_3
   WavePlayer_4
   WavePlayer_5
   WavePlayer_13
   WavePlayer_14
   WavePlayer_15
   WavePlayer_16
   WavePlayer_0
   WavePlayer_6
   WavePlayer_7
  ]
  properties {weights [1,1,1,1,1,1,1,1,1,1,1,1]}
 }
 {id=Random_1, children [WavePlayer_8, WavePlayer_9, WavePlayer_10], properties {weights [1,1,1]}}
 {id=WavePlayer_0, properties {soundWaveAssetPtr=/Game/Sounds/Apollo/Biomes/Generic/AMB_Metal_Stress_Wronk_Groan_Knock_09}}
 {id=WavePlayer_1, properties {soundWaveAssetPtr=/Game/Sounds/Apollo/Biomes/Generic/AMB_Metal_Stress_Wronk_Groan_Knock_01}}
 {id=WavePlayer_10, properties {soundWaveAssetPtr=/Game/Sounds/Apollo/Biomes/Generic/AMB_Metal_Stress_Wronk_Groan_Knock_05}}
 {id=WavePlayer_13, properties {soundWaveAssetPtr=/Game/Sounds/Apollo/Biomes/Generic/AMB_Metal_Stress_Wronk_Groan_Knock_02}}
 {id=WavePlayer_14, properties {soundWaveAssetPtr=/Game/Sounds/Apollo/Biomes/Generic/AMB_Metal_Stress_Wronk_Groan_Stretch_02}}
 {id=WavePlayer_15, properties {soundWaveAssetPtr=/Game/Sounds/Apollo/Biomes/Generic/AMB_Metal_Stress_Wronk_Groan_Stretch_03}}
 {id=WavePlayer_16, properties {soundWaveAssetPtr=/Game/Sounds/Apollo/Biomes/Generic/AMB_Metal_Stress_Wronk_Groan_Stretch_04}}
 {id=WavePlayer_2, properties {soundWaveAssetPtr=/Game/Sounds/Apollo/Biomes/Generic/AMB_Metal_Stress_Wronk_Groan_Knock_06}}
 {id=WavePlayer_3, properties {soundWaveAssetPtr=&AMB_Metal_Stress_Wronk_Groan_Knock_05}}
 {id=WavePlayer_4, properties {soundWaveAssetPtr=/Game/Sounds/Apollo/Biomes/Generic/AMB_Metal_Stress_Wronk_Groan_Knock_04}}
 {id=WavePlayer_5, properties {soundWaveAssetPtr=/Game/Sounds/Apollo/Biomes/Generic/AMB_Metal_Stress_Wronk_Groan_Knock_03}}
 {id=WavePlayer_6, properties {soundWaveAssetPtr=/Game/Sounds/Apollo/Biomes/Generic/AMB_Metal_Stress_Wronk_Groan_Knock_07}}
 {id=WavePlayer_7, properties {soundWaveAssetPtr=/Game/Sounds/Apollo/Biomes/Generic/AMB_Metal_Stress_Wronk_Groan_Knock_08}}
 {id=WavePlayer_8, properties {soundWaveAssetPtr=&AMB_Metal_Stress_Wronk_Groan_Knock_01}}
 {id=WavePlayer_9, properties {soundWaveAssetPtr=&AMB_Metal_Stress_Wronk_Groan_Knock_06}}
]
properties {
 attenuationSettings=/Game/Sounds/Attenuation/Fort_TileAmbient_OneShot_Att
 maxDistance=8000
 pitchMultiplier=0.7
 soundClassObject=&Fort_Ambient_Indoors_Athena
 volumeMultiplier=0.35
 busSends [{audioBus=null, minSendLevel=0, sendLevel=0, soundSourceBus=&AmbientAudio_Exterior_SourceBus}]
}
